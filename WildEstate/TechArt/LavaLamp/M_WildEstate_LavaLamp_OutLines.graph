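# Stylized lava lamp material on the Single Layer Water model, with banded blobs, glass reflection streaks and a height-based liquid fill.
# Camera-facing virtual planes, some offset along the reflection vector, supply UVs for panning noise and reflections; bounds gradients shape top and bottom blobs.
# Rounded masks, a color curve and fill mask drive color, opacity and time-of-day-scaled emissive.
# The normal blends sunlight direction with a camera vector to compromise between lights.
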Material /WildEstate/TechArt/LavaLamp/M_WildEstate_LavaLamp_OutLines
graph {
 properties {
  bAutomaticallySetUsageInEditor=false
  bTangentSpaceNormal=false
  bUseMaterialAttributes=true
  shadingModel=MSM_SingleLayerWater
  usedShadingModels=SingleLayerWater
 }
 nodes [
  n1 = Subtract(
   A=Transform(None=WorldPosition().XYZ) {transformSourceType=TRANSFORMSOURCE_World, transformType=TRANSFORM_Local}
   B=Transform(None=ObjectPositionWS()) {transformSourceType=TRANSFORMSOURCE_World, transformType=TRANSFORM_Local}
  )
  n2 = MaterialFunctionCall() {materialFunction=&ObjectLocalBounds}
  ComponentMask(None=Add(A=Divide(A=@n1, B=@n2.Local Bounds Size)) {constB=0.5}) {b=true}
  n3 = Time()
  Divide(A=Add(A=Sine(None=@n3)))
  n4 = CameraVectorWS()
  n5 = VertexNormalWS()
  n6 = Saturate(None=Power(Base=DotProduct(A=@n4, B=@n5)) {constExponent=1.25})
  n7 = Constant() {r=3}
  Divide(A=Floor(None=Multiply(A=@n6, B=@n7) {constB=5}), B=Subtract(A=@n7))
  n8 = WorldPosition()
  n9 = PreSkinnedLocalBounds()
  n10 = Divide(
   A=Subtract(A=TransformPosition(None=@n8.XYZ) {transformSourceType=TRANSFORMPOSSOURCE_World}, B=@n9.Min)
   B=Multiply(A=@n9.Max, B=ScalarParameter() {defaultValue=1, group=LiquidFill, parameterName=Gradient Size, sortPriority=3})
  )
  n11 = TransformPosition(None=Constant3Vector()) {transformSourceType=TRANSFORMPOSSOURCE_Instance, transformType=TRANSFORMPOSSOURCE_World}
  n12 = Subtract(A=Saturate(None=ComponentMask(None=@n10) {b=true}), B=@n11)
  n13 = ScalarParameter() {defaultValue=-14, group=LiquidFill, parameterName=FillHeight}
  Floor(None=Subtract(A=ComponentMask(None=@n12) {b=true}, B=@n13))
  n14 = VectorParameter() {group=Glass, parameterName=BaseColorReflections, defaultValue=<rgba 0.6795 0.9911 0.1356 1>}
  n15 = VectorParameter() {group=Glass, parameterName=BaseColorOpaqueFill, defaultValue=<rgba 0.3231 0.4969 0.0356 1>}
  n16 = NamedRerouteUsage()
  n17 = LinearInterpolate(
   A=VectorParameter() {group=Glass, parameterName=BaseColorOpaque, defaultValue=<rgba 0.4735 0.6867 0.0908 1>}.RGB
   B=Multiply(A=@n15.RGB, B=ScalarParameter() {defaultValue=0.5229, group=Blobs, parameterName=Blob02_Darken}) {constB=0.5229}
   Alpha=@n16
  )
  n18 = CurveAtlasRowParameter(CurveTime=NamedRerouteUsage()) {
   atlas=/WildEstateTech/Gameplay/Materials/ColorCurves/CA_WildEstate_Tech
   curve=/WildEstateTech/Gameplay/Materials/ColorCurves/CC_WE_LavaLamp
   defaultValue=1
   group=Glass
   parameterName=ColorCurve
  }
  n19 = Multiply(A=@n18, B=@n15.RGB)
  n20 = Saturate(None=Max(A=NamedRerouteUsage(), B=NamedRerouteUsage()))
  n21 = TextureSample(UVs=NamedRerouteUsage()) {texture=/WildEstate/Environment/Textures/T_WildEstate_Windows}
  n22 = NamedRerouteDeclaration(None=LinearInterpolate(A=@n14.RGB, B=LinearInterpolate(A=@n17, B=@n19, Alpha=@n20), Alpha=@n21.R)) {name=BaseColor, nodeColor=<rgba 1 0.375 0 1>}
  LinearInterpolate(A=@n22, B=Multiply(B=@n22), Alpha=@n6)
  n23 = Multiply(A=Multiply(A=@n14.RGB, B=OneMinus(None=@n21.R)), B=ScalarParameter() {defaultValue=0.5, group=Glass, parameterName=EmissiveReflectionLines})
  n24 = Constant() {r=1.2}
  n25 = Constant() {r=1}
  n26 = Constant() {r=1.5}
  n27 = Add(
   A=@n23
   B=Multiply(
    A=NamedRerouteUsage()
    B=MaterialFunctionCall(Value_Morning=@n24, Value_Day=@n25, Value_Evening=@n24, Value_Night=@n26) {materialFunction=&Scalar_Time_of_Day}.Result
   )
  )
  LinearInterpolate(A=@n23, B=@n27, Alpha=NamedRerouteUsage())
  n28 = Constant3Vector() {constant=<rgba 0 0 1 0>}
  LinearInterpolate(B=@n28)
  LinearInterpolate(B=@n28)
  n29 = Max(A=NamedRerouteUsage(), B=Max(A=NamedRerouteUsage(), B=NamedRerouteUsage()))
  n30 = TwoSidedSign()
  LinearInterpolate(A=@n29, B=@n29, Alpha=@n30)
  n31 = NamedRerouteUsage()
  n32 = TextureObject() {samplerType=SAMPLERTYPE_Normal, texture=/Game/Global/Textures/Default/Noise/T_SwirlNoise_N}
  n33 = Constant3Vector() {constant=<rgba 0 0 1 0>}
  n34 = LinearInterpolate(
   A=TextureSample(UVs=Panner(Coordinate=@n31) {speedX=-0.05, speedY=0.025}, Tex=@n32) {samplerType=SAMPLERTYPE_Normal, texture=&T_SwirlNoise_N}.RGB
   B=@n33
  ) {constAlpha=0.85}
  n35 = NamedRerouteUsage()
  n36 = TextureObject() {samplerType=SAMPLERTYPE_LinearGrayscale, texture=/Game/Environments/Asteria/Sets/Lair/Materials/LavaLamp/T_LavaLamp_Background}
  n37 = TextureSample(UVs=Add(A=ComponentMask(None=@n34) {g=true, r=true}, B=Panner(Coordinate=@n35) {speedX=0.002, speedY=-0.015}), Tex=@n36) {samplerType=SAMPLERTYPE_LinearGrayscale, texture=/Game/Global/Textures/Default/Noise/T_SwirlNoise_N_height}
  n38 = Saturate(None=Power(Base=OneMinus(None=@n37.G)) {constExponent=1.5})
  n39 = Add(A=Panner(Coordinate=NamedRerouteUsage()) {speedX=-0.005, speedY=0.023}, B=ComponentMask(None=@n34) {g=true, r=true})
  n40 = Saturate(None=OneMinus(None=TextureSample(UVs=@n39, Tex=@n36) {samplerType=SAMPLERTYPE_LinearGrayscale, texture=&T_SwirlNoise_N_height}.G))
  Max(A=@n40, B=@n38)
  Multiply(A=@n31, B=Constant() {r=1.5})
  n41 = NamedRerouteUsage()
  Multiply(A=@n20, B=@n41)
  Multiply(A=@n16, B=@n41)
  n42 = MaterialFunctionCall() {materialFunction=&BoundingBoxBased_0-1_UVW}
  n43 = NamedRerouteUsage()
  n44 = TextureSample(UVs=Panner(Coordinate=@n43) {speedX=0.01, speedY=0.02}, Tex=@n32) {samplerType=SAMPLERTYPE_Normal, texture=&T_SwirlNoise_N}
  n45 = Multiply(
   A=Add(A=@n42.RGB, B=ScalarParameter() {defaultValue=0.45, group=Blobs, parameterName=BottomBlobOffset}) {constB=0.2}
   B=NamedRerouteDeclaration(None=LinearInterpolate(A=@n44.RGB, B=@n33) {constAlpha=0.6}) {name=Noise_Down, nodeColor=<rgba 0.6172 0 1 1>}
  )
  n46 = OneMinus(None=Power(Base=ComponentMask(None=@n45) {b=true}) {constExponent=16})
  n47 = Add(A=@n42.RGB) {constB=0.033}
  n48 = Panner(Coordinate=@n43) {speedX=-0.03, speedY=-0.02}
  n49 = NamedRerouteDeclaration(None=LinearInterpolate(A=TextureSample(UVs=@n48, Tex=@n32) {samplerType=SAMPLERTYPE_Normal, texture=&T_SwirlNoise_N}.RGB, B=@n33) {constAlpha=0.85}) {name=Noise_UP, nodeColor=<rgba 0 1 0.2969 1>}
  n50 = Power(Base=ComponentMask(None=Multiply(A=@n47, B=@n49)) {b=true}) {constExponent=16}
  n51 = Max(A=Max(A=Saturate(None=@n46), B=Saturate(None=@n50)), B=@n40)
  n52 = Constant() {r=3}
  Multiply(A=Round(None=Multiply(A=@n51, B=@n52) {constB=5})) {constB=0.2}
  n53 = Constant3Vector() {constant=<rgba 0 0 1 0>}
  n54 = Constant3Vector() {constant=<rgba 1 0 0 0>}
  n55 = Add(A=@n11, B=AppendVector(A=Constant2Vector(), B=@n13))
  n56 = ScalarParameter() {defaultValue=39.6541, group=LiquidFill, parameterName=FillRadius}
  n57 = NamedRerouteUsage()
  n58 = LinearInterpolate(A=@n57, B=Constant3Vector() {constant=<rgba 0 0 1 0>}) {constAlpha=0.25}
  n59 = SphereMask(
   A=Add(
    A=MaterialFunctionCall!(Plane Normal Axis=@n53, Plane X Axis=@n54, Plane Center=@n55, UV Scale=@n56) {materialFunction=/Game/Athena/Environments/Nexus/Materials/VirtualPlaneCoordinatesAbs}.UVs
    B=ComponentMask(None=@n58) {g=true, r=true}
   )
   B=Constant2Vector() {g=0.5, r=0.5}
  ) {attenuationRadius=0.5}
  n60 = Subtract(A=ComponentMask(None=Add(A=@n57, B=@n12)) {b=true}, B=@n13)
  n61 = OneMinus(None=Saturate(None=Floor(None=@n60)))
  NamedRerouteDeclaration(None=Saturate(None=Add(A=@n59, B=@n61))) {name=FillMask, nodeColor=<rgba 0.9219 1 0 1>}
  n62 = CollectionParameter() {
   collection=&FortniteMaterialParameters
   parameterName=SunlightDirectionalVector
  }
  n63 = Constant3Vector() {constant=<rgba 0 0 -1 0>}
  n64 = Normalize(VectorInput=Add(A=ComponentMask(None=@n62) {b=true, g=true, r=true}, B=Transform(None=@n63) {transformSourceType=TRANSFORMSOURCE_Camera}) {desc=Compromise for Direction light and all othe rlights})
  NamedRerouteDeclaration(None=MaterialProxyReplace(Realtime=@n64, MaterialProxy=Transform(None=@n64) {transformSourceType=TRANSFORMSOURCE_World, transformType=TRANSFORM_Tangent})) {name=Normal, nodeColor=<rgba 1 0 0.2344 1>}
  n65 = WorldPosition()
  n66 = ComponentMask(None=@n65.XYZ) {r=true}
  n67 = ComponentMask(None=@n65.XYZ) {b=true, g=true}
  n68 = ScalarParameter() {defaultValue=3, group=Blobs, parameterName=Blob 01 Rotation Speed X}
  n69 = ScalarParameter() {defaultValue=-6, group=Blobs, parameterName=Blob 01 Rotation Speed Y}
  n70 = AppendVector(A=@n66, B=Panner(Coordinate=@n67, Speed=AppendVector(A=@n68, B=@n69)))
  n71 = Constant() {r=6}
  n72 = ScalarParameter() {defaultValue=1, group=Blobs, parameterName=Blob 02 Rotation Speed X}
  n73 = ScalarParameter() {defaultValue=2, group=Blobs, parameterName=Blob 02 Rotation Speed Y}
  n74 = AppendVector(A=@n66, B=Panner(Coordinate=@n67, Speed=AppendVector(A=@n72, B=@n73)))
  n75 = Constant() {r=6}
  n76 = Max(
   A=Power(Base=Noise!(World Position=@n70, FilterWidth=@n71) {levels=3, noiseFunction=NOISEFUNCTION_GradientTex3D, outputMin=-0.5, scale=0.01})
   B=Power(Base=Noise(World Position=@n74, FilterWidth=@n75) {levels=3, noiseFunction=NOISEFUNCTION_GradientTex3D, outputMin=-0.5, scale=0.125})
  )
  n77 = Constant() {r=3}
  n78 = Divide(A=Floor(None=Multiply(A=@n76, B=@n77) {constB=5}), B=Subtract(A=@n77))
  NamedRerouteDeclaration(None=Multiply(A=@n78, B=Constant3Vector() {constant=<rgba 0.314 0.491 0.0319 1>})) {name=TestColor, nodeColor=<rgba 1 0 0.4453 1>}
  n79 = Saturate(None=@n78)
  NamedRerouteDeclaration(None=@n79) {name=TestAlpha, nodeColor=<rgba 0 1 0.4844 1>}
  n80 = Multiply(A=@n51, B=@n52) {constB=5}
  NamedRerouteDeclaration(None=Divide(A=Floor(None=@n80), B=Subtract(A=@n52))) {name=BlobGradient, nodeColor=<rgba 1 0.3047 0 1>}
  n81 = Fresnel(ExponentIn=ScalarParameter() {defaultValue=1.4086, group=LiquidFill, parameterName=FillFresnelEdge}, Normal=VertexNormalWS()) {baseReflectFraction=0}
  NamedRerouteDeclaration(None=OneMinus(None=Round(None=@n81))) {name=fresnelMask, nodeColor=<rgba 0.4531 0 1 1>}
  NamedRerouteDeclaration(None=Round(None=@n38)) {name=Blob02Alpha, nodeColor=<rgba 0.0312 1 0 1>}
  NamedRerouteDeclaration(None=@n6) {name=EmissiveFresnel, nodeColor=<rgba 1 0.4453 0 1>}
  n82 = Constant3Vector()
  n83 = MaterialFunctionCall(Float3=Divide(A=TransformPosition(None=@n82) {transformSourceType=TRANSFORMPOSSOURCE_Instance, transformType=TRANSFORMPOSSOURCE_World}) {constB=100}) {materialFunction=&BreakOutFloat3Components}
  n84 = Frac(None=Add(A=Add(A=@n83.R, B=@n83.R), B=@n83.B))
  n85 = Transform(None=Constant3Vector() {constant=<rgba 0 0 1 0>}) {transformSourceType=TRANSFORMSOURCE_Camera}
  n86 = Transform(None=Constant3Vector() {constant=<rgba 1 0 0 0>}) {transformSourceType=TRANSFORMSOURCE_Camera}
  n87 = ObjectPositionWS()
  n88 = ReflectionVectorWS()
  n89 = Add(A=@n87, B=Multiply(A=Multiply(A=@n88) {constB=-1}, B=ScalarParameter() {defaultValue=-15, group=Glass, parameterName=GlassRelectionOffset}))
  n90 = ScalarParameter() {defaultValue=150, group=Glass, parameterName=GlassRelectionTiling}
  NamedRerouteDeclaration(None=Add(
    A=Multiply(A=@n84) {constB=33.3}
    B=MaterialFunctionCall(Plane Normal Axis=@n85, Plane X Axis=@n86, Plane Center=@n89, UV Scale=@n90) {materialFunction=&VirtualPlaneCoordinates}.UVs
   )) {name=Reflection UVs, nodeColor=<rgba 0 0.5547 1 1>}
  NamedRerouteDeclaration(None=MaterialFunctionCall(
    Plane Normal Axis=@n85
    Plane X Axis=@n86
    Plane Center=@n87
    UV Scale=ScalarParameter() {defaultValue=150, group=Blobs, parameterName=BLobNoiseScale}
   ) {materialFunction=&VirtualPlaneCoordinates}.UVs) {name=BlobNoiseUV, nodeColor=<rgba 0 0.9766 1 1>}
  n91 = ReflectionVectorWS()
  n92 = Add(A=Multiply(A=Multiply(A=@n91) {constB=-1}, B=ScalarParameter() {defaultValue=-10, group=Blobs, parameterName=Blob02_Offset}), B=@n87)
  n93 = ScalarParameter() {defaultValue=200, group=Blobs, parameterName=Blob02_Scale}
  NamedRerouteDeclaration(None=Add(
    A=Constant2Vector() {g=0.33, r=0.125}
    B=MaterialFunctionCall(Plane Normal Axis=@n85, Plane X Axis=@n86, Plane Center=@n92, UV Scale=@n93) {materialFunction=&VirtualPlaneCoordinates}.UVs
   )) {name=Blob02_UV, nodeColor=<rgba 0.5 1 0 1>}
  n94 = ScalarParameter() {defaultValue=-5, group=Blobs, parameterName=Blob01_Offset}
  NamedRerouteDeclaration(None=MaterialFunctionCall(
    Plane Normal Axis=@n85
    Plane X Axis=@n86
    Plane Center=Add(A=@n94, B=@n87)
    UV Scale=ScalarParameter() {defaultValue=200, group=Blobs, parameterName=Blob01_Scale}
   ) {materialFunction=&VirtualPlaneCoordinates}.UVs) {name=Blob01_UV, nodeColor=<rgba 0.5 1 0 1>}
  NamedRerouteDeclaration(None=Round(None=Saturate(None=@n51))) {name=BlobFinal, nodeColor=<rgba 1 0.4453 0 1>}
  Round(None=@n50)
  Round(None=@n46)
  Round(None=@n79)
  ScalarParameter() {defaultValue=-14, group=LiquidFill, parameterName=CenterFillHeight}
  n95 = Saturate(None=Multiply(A=@n29, B=@n30))
  n96 = NamedRerouteUsage()
  SetMaterialAttributes!(
   MaterialAttributes
   Base Color=@n22
   Specular=Constant()
   Roughness=Constant()
   Emissive Color=@n27
   Opacity=LinearInterpolate(
    A=ScalarParameter() {defaultValue=0.672, group=Glass, parameterName=Opacity}
    B=ScalarParameter() {defaultValue=1000, group=Glass, parameterName=Opacity_Fill}
    Alpha=Multiply(A=@n95, B=@n96)
   )
   Normal=NamedRerouteUsage()
  ) {attributeSetTypes [
    #1
    #2
    #3
    #4
    #5
    #6
   ]}
  n97 = Constant()
  SingleLayerWaterMaterialOutput!(ScatteringCoefficients=@n97, AbsorptionCoefficients=@n97, PhaseG, ColorScaleBehindWater)
  SmoothStep()
  VertexColor()
  WorldPosition() {worldPositionShaderOffset=WPT_ExcludeAllShaderOffsets}
 ]
}
